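# Material function for the Cactus Dancer cosmetic body that adds animated glowing lines and audio-reactive glowing parts, behind a static switch.
# Time and a power-warped mask gradient drive cosine line pulses, while pre-skinned height normalized by local bounds builds a cyan-to-yellow gradient.
# MF_AudioReactivity supplies spectrum and amplitude, with thresholded four-channel bands and masks for shirt, smiley and pants effects.
# Two BlendMaterialAttributes stages layer lines, then reactive emissive, onto the material attributes.
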MaterialFunction /BRCosmetics/Characters/Player/Male/Medium/Bodies/M_MED_Cactus_Dancer/Materials/MF_CactusDancer_Lines
graph {nodes [
  n1 = Time()
  n2 = Constant() {r=0.5}
  Add(A=Multiply(A=Sine(None=@n1), B=@n2), B=@n2)
  n3 = NamedRerouteUsage()
  n4 = NamedRerouteUsage()
  n5 = NamedRerouteUsage()
  n6 = BlendMaterialAttributes(
   A=@n3
   B=SetMaterialAttributes!(MaterialAttributes=@n3, BaseColor=@n4, EmissiveColor=Multiply(A=@n4, B=@n5) {constB=20}) {attributeSetTypes [#1, #2]}
   Alpha=Multiply(A=NamedRerouteUsage(), B=NamedRerouteUsage())
  ) {desc=dancer lines}
  n7 = Add(A=NamedRerouteUsage(), B=Multiply(A=NamedRerouteUsage()) {constB=0.25})
  n8 = NamedRerouteUsage()
  n9 = SetMaterialAttributes!(MaterialAttributes=NamedRerouteUsage(), EmissiveColor=Multiply(A=NamedRerouteUsage(), B=Add(A=@n7, B=@n8))) {attributeSetTypes [#2]}
  n10 = Multiply(A=Add(A=NamedRerouteUsage(), B=NamedRerouteUsage()), B=NamedRerouteUsage())
  FunctionOutput(None=StaticSwitchParameter(True=BlendMaterialAttributes(A=@n6, B=@n9, Alpha=@n10) {desc=reactive glowing parts}, False=NamedRerouteUsage()) {group=Skin FX, parameterName=Use Cactus Dancer Lines})
  n11 = NamedRerouteUsage()
  MaterialFunctionCall(Input=@n11, Input Low=Constant(), Input High=Constant() {r=0.11}) {desc=smiley mask outer, materialFunction=&RemapValue_0-1}
  Multiply(A=NamedRerouteUsage()) {constB=0.35}
  n12 = FunctionInput() {bUsePreviewValueAsDefault=true, inputName=SkinFX_Mask2, inputType=FunctionInput_Scalar, sortPriority=2, previewValue=<1 1 0 1>}
  NamedRerouteDeclaration(None=@n12) {name=LineGradient, nodeColor=<rgba 0 0.0859 1 1>}
  n13 = Saturate(None=ComponentMask(None=NamedRerouteUsage()) {g=true})
  n14 = Constant()
  n15 = Constant() {r=0.3}
  n16 = Saturate(None=Add(
    A=OneMinus(None=@n13)
    B=MaterialFunctionCall(Input=@n11, Input Low=@n14, Input High=@n15) {desc=smiley mask, materialFunction=&RemapValue_0-1}.Result
   ))
  n17 = MaterialFunctionCall(Input=@n11, Input Low=Constant(), Input High=Constant() {r=0.11}) {desc=tshirt mask, materialFunction=&RemapValue_0-1}
  NamedRerouteDeclaration(None=Multiply(A=OneMinus(None=@n16), B=NamedRerouteDeclaration(None=@n17.Result) {name=ShirtMask, nodeColor=<rgba 0.4531 1 0 1>})) {name=TShirtSmileyGlow, nodeColor=<rgba 1 0 0.5156 1>}
  n18 = NamedRerouteUsage()
  n19 = ScalarParameter() {defaultValue=2, group=Cactus Dancer Lines, parameterName=CD_LineExp, sliderMax=16}
  n20 = ScalarParameter() {defaultValue=5, group=Cactus Dancer Lines, parameterName=CD_LineAmount}
  n21 = NamedRerouteUsage()
  n22 = Time()
  n23 = ScalarParameter() {defaultValue=1.333, parameterName=CD_TimePeriod, sliderMax=3}
  n24 = Add(
   A=Multiply(A=Power(Base=@n18, Exp=@n19), B=@n20)
   B=If(A=Saturate(None=@n21), "A > B"=Divide(A=@n22) {constB=0.333}, "A < B"=Divide(A=@n22, B=@n23) {constB=1.666}) {constB=0.1}
  )
  NamedRerouteDeclaration(None=Fmod(A=Add(A=@n24) {constB=0.5}, B=@n20)) {name=Modulo, nodeColor=<rgba 1 0.5859 0 1>}
  n25 = MaterialFunctionCall(Float4=NamedRerouteUsage()) {materialFunction=&BreakOutFloat4Components}
  n26 = NamedRerouteUsage()
  n27 = Saturate(None=Ceil(None=Subtract(A=@n26) {constB=4})) {desc=Band1}
  n28 = Ceil(None=Subtract(A=@n26) {constB=3})
  n29 = Multiply(A=OneMinus(None=@n27), B=@n28) {desc=Band2}
  n30 = OneMinus(None=Saturate(None=@n28))
  n31 = Ceil(None=Subtract(A=@n26) {constB=2})
  n32 = Add(A=Add(A=Multiply(A=@n25.R, B=@n27), B=Multiply(A=@n25.G, B=@n29)), B=Multiply(A=@n25.B, B=Multiply(A=@n30, B=@n31) {desc=Band3}))
  n33 = Saturate(None=@n31)
  n34 = Subtract(A=@n26)
  n35 = Multiply(A=@n25.A, B=Multiply(A=OneMinus(None=@n33), B=Ceil(None=@n34)) {desc=Band4})
  n36 = Multiply(A=Saturate(None=Add(A=@n32, B=@n35))) {constB=4}
  NamedRerouteDeclaration(None=Saturate(None=Add(A=@n36) {constB=0.25})) {name=BandsAnimaiton, nodeColor=<rgba 1 0.2109 0 1>}
  n37 = PreSkinnedPosition()
  NamedRerouteDeclaration(None=ShaderStageSwitch(PixelShader=VertexInterpolator(VS=@n37).PS, VertexShader=@n37)) {name=PreSkinnedLocalPos, nodeColor=<rgba 1 0.4922 0 1>}
  n38 = FunctionInput() {bUsePreviewValueAsDefault=true, inputName=SkinFX_Mask, inputType=FunctionInput_Vector4, sortPriority=1, previewValue=<1 1 0 1>}
  n39 = ChannelMaskParameter(None=@n38) {
   group=Cactus Dancer Lines
   maskChannel=Blue
   parameterName=Cactus Dancer Line Distortion Channel
   sortPriority=0
   defaultValue=<rgba 0 0 1 0>
  }
  NamedRerouteDeclaration(None=@n39.R) {name=FX_TShirtMask, nodeColor=<rgba 0 0.25 1 1>}
  n40 = ComponentMask(None=NamedRerouteUsage()) {b=true}
  n41 = ComponentMask(None=PreSkinnedLocalBounds().Extents) {b=true}
  n42 = Add(
   A=Multiply(A=Divide(A=@n40, B=@n41), B=ScalarParameter() {defaultValue=6.8704, group=Cactus Dancer Lines, parameterName=CD_Scale}) {constB=4}
   B=ScalarParameter() {defaultValue=-5.2285, group=Cactus Dancer Lines, parameterName=CD_Bias}
  ) {constB=-2.5}
  NamedRerouteDeclaration(None=LinearInterpolate(A=Constant3Vector() {constant=<rgba 0 1 0.9463 0>}, B=Constant3Vector() {constant=<rgba 1 0.9385 0 0>}, Alpha=Saturate(None=@n42))) {name=BC_Out, nodeColor=<rgba 1 0.0938 0 1>}
  NamedRerouteDeclaration(None=MaterialFunctionCall(Input=@n11, Input Low=Constant() {r=0.15}, Input High=Constant() {r=1}) {materialFunction=&RemapValue_0-1}.Result) {name=Shirt_AlphaMask, nodeColor=<rgba 0 1 0.6719 1>}
  n43 = NamedRerouteUsage()
  n44 = Constant() {r=0.5}
  n45 = Add(A=Multiply(A=Cosine(None=@n24), B=@n44) {constB=0.5}, B=@n44) {constB=0.5}
  n46 = NamedRerouteUsage()
  n47 = Saturate(None=Power(Base=Multiply(A=@n45, B=@n46), Exp=LinearInterpolate(Alpha=@n46) {constA=1, constB=2.5}) {constExponent=2.5})
  n48 = NamedRerouteUsage()
  NamedRerouteDeclaration(None=If(A=Saturate(None=@n43), "A > B"=Multiply(A=@n47, B=@n48), "A < B"=@n47) {constB=0.2}) {name=LineAlpha, nodeColor=<rgba 0.4627 0 1 1>}
  n49 = MaterialFunctionCall() {materialFunction=/Game/Packages/Fortress_SharedMaterials/Base_Material_Functions/BaseCharacterShaderFunctions/Audio/MF_AudioReactivity}
  NamedRerouteDeclaration(None=@n49.Spectrum) {name=Audio_Spectrum, nodeColor=<rgba 1 0 0.6094 1>}
  n50 = ChannelMaskParameter(None=@n38) {group=Cactus Dancer Lines, maskChannel=Alpha, parameterName=Cactus Dancer Glow Channel, sortPriority=0, defaultValue=<rgba 0 0 0 1>}
  NamedRerouteDeclaration(None=@n50.R) {name=GlowMask, nodeColor=<rgba 1 0.5523 0.2764 1>}
  n51 = FunctionInput() {inputName=MA, inputType=FunctionInput_MaterialAttributes}
  NamedRerouteDeclaration(None=GetMaterialAttributes(None=NamedRerouteDeclaration(None=@n51) {name=MA_In, nodeColor=<rgba 1 0.5 0 1>}) {attributeGetTypes [#1]}.BaseColor) {name=BC_In, nodeColor=<rgba 0 1 0.4141 1>}
  NamedRerouteDeclaration(None=@n49.Amplitude) {name=Audio_Amplitude, nodeColor=<rgba 1 0 0.8438 1>}
  n52 = MaterialFunctionCall(Float4=NamedRerouteUsage()) {materialFunction=&BreakOutFloat4Components}
  n53 = Add(A=@n52.R, B=@n52.R)
  n54 = Multiply(A=Add(A=Add(A=@n53, B=@n52.B), B=@n52.A)) {constB=0.33}
  n55 = Multiply(A=NamedRerouteUsage()) {constB=0.9}
  NamedRerouteDeclaration(None=Ceil(None=Subtract(A=@n54, B=@n55))) {name=SpecturmAnimPants, nodeColor=<rgba 1 0.2578 0 1>}
 ]}
