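# Physics asset for the male medium Cactus Dancer cosmetic body, setting up jointed dynamic bones for its leg straps and necklace.
# Capsule bodies cover the thighs, calves and upper spine, with each calf constrained rigidly to its thigh.
# A three-link strap chain hangs from each thigh with one swing axis free.
# A four-link necklace chain hangs from spine_05, with its first link more tightly limited than the rest.
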PhysicsAsset /BRCosmetics/Characters/Player/Male/Medium/Bodies/M_MED_Cactus_Dancer/Meshes/M_MED_Cactus_Dancer_Physics
bodies {
 calf_l [{shapeType=Capsule, length=20, radius=5.9, shapeName=calf_l_capsule, center=<-7 0 0.4>, rotation=<rot 90 0 0>}]
 calf_r [{shapeType=Capsule, length=20, radius=5.9, shapeName=calf_l_capsule, center=<7 0 -0.4>, rotation=<rot 90 0 0>}]
 dyn_L_legStrap_1 [{shapeType=Capsule, length=4, radius=2, shapeName=dyn_L_legStrap_1_capsule, center=<2 0 0>, rotation=<rot 90 0 0>}]
 dyn_L_legStrap_2 [{shapeType=Capsule, length=4, radius=2.6, shapeName=dyn_L_legStrap_2_capsule, center=<2 0 0>, rotation=<rot 90 0 0>}]
 dyn_L_legStrap_3 [{shapeType=Capsule, length=4, radius=3, shapeName=dyn_L_legStrap_3_capsule, center=<2 0 0>, rotation=<rot 90 0 0>}]
 dyn_R_legStrap_1 [{shapeType=Capsule, length=4, radius=2, shapeName=dyn_R_legStrap_1_capsule, center=<-2 0 0>, rotation=<rot 90 0 0>}]
 dyn_R_legStrap_2 [{shapeType=Capsule, length=4, radius=2.6, shapeName=dyn_R_legStrap_2_capsule, center=<-2 0 0>, rotation=<rot 90 0 0>}]
 dyn_R_legStrap_3 [{shapeType=Capsule, length=4, radius=3, shapeName=dyn_R_legStrap_3_capsule, center=<-2 0 0>, rotation=<rot 90 0 0>}]
 dyn_necklace_1 [{shapeType=Capsule, length=2, radius=4, shapeName=dyn_necklace_1_capsule, center=<1 0 0>, rotation=<rot 90 0 0>}]
 dyn_necklace_2 [{shapeType=Capsule, length=2, radius=4, shapeName=dyn_necklace_2_capsule, center=<1 0 0>, rotation=<rot 90 0 0>}]
 dyn_necklace_3 [{shapeType=Capsule, length=2, radius=4, shapeName=dyn_necklace_3_capsule, center=<1.5 0 0>, rotation=<rot 90 0 0>}]
 dyn_necklace_4 [{shapeType=Capsule, length=2, radius=4, shapeName=dyn_necklace_4_capsule, center=<2 0 0>, rotation=<rot 90 0 0>}]
 spine_05 [{shapeType=Capsule, length=10, radius=11.5, shapeName=spine_05_capsule, center=<-4.1441 -3.8569 0>}]
 thigh_l [{shapeType=Capsule, length=20, radius=9, shapeName=thigh_l_capsule, center=<-21 0 0>, rotation=<rot 90 0 0>}]
 thigh_r [{shapeType=Capsule, length=20, radius=9, shapeName=thigh_l_capsule, center=<21 0 0>, rotation=<rot 90 0 0>}]
}
constraints [
 {bone1Name=dyn_L_legStrap_1, bone2Name=thigh_l, swing1LimitDegrees=30, swing2LimitDegrees=60, swing2Motion=Free, twistLimitDegrees=0}
 {bone1Name=dyn_L_legStrap_2, bone2Name=dyn_L_legStrap_1, swing1LimitDegrees=30, swing2LimitDegrees=60, swing2Motion=Free, twistLimitDegrees=0}
 {bone1Name=dyn_L_legStrap_3, bone2Name=dyn_L_legStrap_2, swing1LimitDegrees=30, swing2LimitDegrees=60, swing2Motion=Free, twistLimitDegrees=0}
 {bone1Name=dyn_R_legStrap_2, bone2Name=dyn_R_legStrap_1, swing1LimitDegrees=30, swing2LimitDegrees=60, swing2Motion=Free, twistLimitDegrees=0}
 {bone1Name=dyn_R_legStrap_1, bone2Name=thigh_r, swing1LimitDegrees=30, swing2LimitDegrees=60, swing2Motion=Free, twistLimitDegrees=0}
 {bone1Name=dyn_R_legStrap_3, bone2Name=dyn_R_legStrap_2, swing1LimitDegrees=30, swing2LimitDegrees=60, swing2Motion=Free, twistLimitDegrees=0}
 {bone1Name=dyn_necklace_1, bone2Name=spine_05, swing1LimitDegrees=0, swing2LimitDegrees=80, twistLimitDegrees=0}
 {bone1Name=dyn_necklace_2, bone2Name=dyn_necklace_1, swing1LimitDegrees=40, swing2LimitDegrees=80, twistLimitDegrees=0}
 {bone1Name=dyn_necklace_3, bone2Name=dyn_necklace_2, swing1LimitDegrees=40, swing2LimitDegrees=80, twistLimitDegrees=0}
 {bone1Name=dyn_necklace_4, bone2Name=dyn_necklace_3, swing1LimitDegrees=40, swing2LimitDegrees=80, twistLimitDegrees=0}
 {
  bone1Name=calf_r
  bone2Name=thigh_r
  swing1LimitDegrees=45
  swing1Motion=Locked
  swing2LimitDegrees=45
  swing2Motion=Locked
  twistLimitDegrees=45
  twistMotion=Locked
 }
 {
  bone1Name=calf_l
  bone2Name=thigh_l
  swing1LimitDegrees=45
  swing1Motion=Locked
  swing2LimitDegrees=45
  swing2Motion=Locked
  twistLimitDegrees=45
  twistMotion=Locked
 }
]
