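# Translucent unlit Niagara sprite and mesh-particle material for a blue ray effect on the Cactus Dancer character skin.
# Panning noise with zeroed V tiling forms streaks, shaped by UV edge fades, a horizontal sphere mask and a panning mask texture.
# Opacity adds particle alpha, view-angle fade and depth fade on deferred shading only.
# Particle color is scaled by time of day, with an optional top-down gradient for color and opacity.
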Material /BRCosmetics/Characters/Player/Male/Medium/Bodies/M_MED_Cactus_Dancer/Skin/FX/M_TigerRootHype_Blue_Ray
graph {
 properties {
  bUsedWithNiagaraMeshParticles=true
  bUsedWithNiagaraSprites=true
  blendMode=BLEND_Translucent
  pixelDepthOffsetMode=PDOM_Legacy
  shadingModel=MSM_Unlit
  usedShadingModels=Unlit
 }
 nodes [
  n1 = ParticleColor()
  n2 = ScalarParameter() {defaultValue=1, parameterName=TextureTile}
  n3 = TextureCoordinate() {vTiling=0}
  n4 = ScalarParameter() {defaultValue=-1, parameterName=Pan Speed}
  n5 = Constant()
  n6 = TextureSampleParameter2D(UVs=Panner(Coordinate=Multiply(A=@n2, B=@n3), Speed=AppendVector(A=@n4, B=@n5)) {speedX=-1}) {parameterName=Tex, samplerType=SAMPLERTYPE_Grayscale, texture=&LowResBlurredNoise}
  n7 = ScalarParameter() {defaultValue=1.35, parameterName=Texture_Power}
  n8 = TextureCoordinate()
  n9 = Multiply(A=Multiply(A=@n8, B=OneMinus(None=@n8))) {constB=4}
  n10 = Multiply(A=Multiply(A=ComponentMask(None=@n9) {r=true}, B=ComponentMask(None=@n9) {g=true})) {constB=4}
  n11 = ComponentMask(None=@n8) {g=true}
  n12 = LinearInterpolate(A=@n11, B=Multiply(A=@n11, B=OneMinus(None=@n11)), Alpha=ScalarParameter() {parameterName=FadeBottom})
  n13 = ScalarParameter() {defaultValue=2, parameterName=FadeTop}
  n14 = Multiply(
   A=Multiply(A=Power(Base=@n6.R, Exp=@n7) {constExponent=1.35}, B=ScalarParameter() {defaultValue=3, parameterName=Texture_Multiply}) {constB=3}
   B=Multiply(A=Saturate(None=@n10), B=Power(Base=@n12, Exp=@n13))
  )
  n15 = ScalarParameter() {defaultValue=-1, parameterName=MaskSpeed}
  n16 = Constant()
  n17 = TextureSampleParameter2D(UVs=Panner(Coordinate=TextureCoordinate() {vTiling=0.9}, Speed=AppendVector(A=@n15, B=@n16)) {speedX=-1}) {parameterName=Mask, texture=&/T_TigerRootHype_Blue_Ray_Mask}
  n18 = Multiply(
   A=SphereMask(A=TextureCoordinate() {vTiling=0}, B=Constant2Vector() {r=0.5}, Hardness=ScalarParameter() {parameterName=Hardness}) {attenuationRadius=0.5, hardnessPercent=0}
   B=@n17.R
  ) {constA=1}
  n19 = ScalarParameter() {defaultValue=0.5, parameterName="FadeSensitivity (optional) (S)"}
  n20 = Multiply(
   A=@n1.A
   B=Multiply(
    A=Multiply(A=@n14, B=@n18)
    B=MaterialFunctionCall("FadeSensitivity (optional)"=@n19) {materialFunction=&FadeBasedOnViewAngle}.Result
   )
  )
  n21 = DepthFade(Opacity=@n20, FadeDistance=ScalarParameter() {parameterName=DepthFade_Distance})
  n22 = ShadingPathSwitch(Default=@n21, Deferred=@n21, Forward=@n20, Mobile=@n20)
  n23 = ScalarParameter() {group=Top Down Gradient, parameterName=OPACITY MASK OFFSET}
  n24 = SmoothStep(
   Min=@n23
   Max=Add(A=@n23, B=ScalarParameter() {group=Top Down Gradient, parameterName=SMOOTHSTEP OFFSET})
   Value=ComponentMask(None=TextureCoordinate()) {g=true}
  )
  n25 = Multiply(A=@n22, B=LinearInterpolate(Alpha=Saturate(None=@n24)))
  If(A=ScalarParameter() {group=Top Down Gradient, parameterName=USE TOP DOWN GRADIENT}, "A > B"=@n25, "A == B"=@n25, "A < B"=@n22) {constB=1}
  n26 = Multiply(
   A=MaterialFunctionCall(
    Value_Morning=ScalarParameter() {defaultValue=1, group=TOD, parameterName=TOD_Morning}
    Value_Day=ScalarParameter() {defaultValue=1, group=TOD, parameterName=TOD_Day}
    Value_Evening=ScalarParameter() {defaultValue=1, group=TOD, parameterName=TOD_Evening}
    Value_Night=ScalarParameter() {defaultValue=0.75, group=TOD, parameterName=TOD_Night}
   ) {materialFunction=&Scalar_Time_of_Day}.Result
   B=@n1.RGB
  )
  n27 = ScalarParameter() {group=Top Down Gradient, parameterName=MASK OFFSET}
  n28 = Add(A=@n27, B=ScalarParameter() {group=Top Down Gradient, parameterName=SMOOTHSTEP OFFSET})
  n29 = ComponentMask(None=TextureCoordinate()) {g=true}
  n30 = LinearInterpolate(
   A=VectorParameter() {group=Top Down Gradient, parameterName=TOP COLOR, defaultValue=<rgba 0 0 0 1>}.RGB
   B=Multiply(A=ScalarParameter() {group=Top Down Gradient, parameterName=BOTTOM COLOR MULT}, B=@n26)
   Alpha=Saturate(None=SmoothStep(Min=@n27, Max=@n28, Value=@n29))
  )
  MaterialFunctionCall(Emissive=If(A=ScalarParameter() {group=Top Down Gradient, parameterName=USE TOP DOWN GRADIENT}, "A > B"=@n30, "A == B"=@n30, "A < B"=@n26) {constB=1}) {materialFunction=&MF_FX_EmissiveAdjustments}
 ]
}
